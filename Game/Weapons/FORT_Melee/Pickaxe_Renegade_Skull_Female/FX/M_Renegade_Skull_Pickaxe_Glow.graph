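# Two-sided translucent unlit glow material for the female Renegade Skull pickaxe FX.
# A Fresnel mask tints GlowColor or particle color, scaled by a dynamic Emissive parameter, and drives opacity with depth fade (bypassed on mobile) and particle alpha.
# An offset branch combines switchable camera offset and vertex-normal push.
# InvertFresnel? blends two identical Fresnel outputs, so it has no effect.
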Material /Game/Weapons/FORT_Melee/Pickaxe_Renegade_Skull_Female/FX/M_Renegade_Skull_Pickaxe_Glow
graph {
 properties {
  bUsedWithMeshDeformer=true
  bUsedWithNiagaraMeshParticles=true
  bUsedWithSkeletalMesh=true
  blendMode=BLEND_Translucent
  pixelDepthOffsetMode=PDOM_Legacy
  shadingModel=MSM_Unlit
  twoSided=true
  usedShadingModels=Unlit
 }
 nodes [
  Constant() {r=4}
  n1 = ScalarParameter() {defaultValue=-6, parameterName=CameraOffsetAmount}
  n2 = VertexNormalWS()
  n3 = StaticSwitchParameter(
   True=Add(
    A=MaterialFunctionCall(Offset Amount=@n1) {materialFunction=/Engine/Functions/Engine_MaterialFunctions02/WorldPositionOffset/CameraOffset}.Camera Offset
    B=Multiply(A=@n2) {constB=2}
   )
   False=Constant()
  ) {defaultValue=true, parameterName=UseCameraOffset}
  n4 = VertexNormalWS()
  n5 = ScalarParameter() {defaultValue=1, parameterName=VertexNormalAmount}
  LinearInterpolate(A=@n3, B=Add(A=@n3, B=Multiply(A=@n4, B=@n5)), Alpha=ScalarParameter() {parameterName=UseNormalMult?})
  n6 = MaterialFunctionCall(
   Brightness=ScalarParameter() {defaultValue=1, parameterName=Fresnel_Brightness}
   Exponent=ScalarParameter() {defaultValue=4, parameterName=Fresnel}
   Camera Vector=CameraVectorWS()
  ) {materialFunction=&MF_Fresnel_FX}
  n7 = LinearInterpolate(
   A=Constant() {r=1}
   B=LinearInterpolate(A=@n6.Fresnel Double Sided, B=@n6.Fresnel Double Sided, Alpha=ScalarParameter() {parameterName=InvertFresnel?})
   Alpha=ScalarParameter() {defaultValue=1, parameterName=UseFresnel?}
  ) {constAlpha=0}
  n8 = Multiply(A=VectorParameter() {parameterName=GlowColor, defaultValue=<rgba 0.0868 0.7766 2 1>}.RGB, B=@n7)
  n9 = ParticleColor()
  n10 = Multiply(A=@n9.RGB, B=@n7)
  n11 = ScalarParameter() {parameterName=UseParticleColor}
  MaterialFunctionCall(Emissive=Multiply(A=DynamicParameter() {paramNames [Emissive,Param2,Param3,Param4]}.Emissive, B=LinearInterpolate(A=@n8, B=@n10, Alpha=@n11))) {materialFunction=&MF_FX_EmissiveAdjustments}
  n12 = DepthFade(Opacity=@n7, FadeDistance=ScalarParameter() {defaultValue=1, parameterName=FadeDistance}) {fadeDistanceDefault=1}
  Multiply(
   A=Multiply(
    A=ShadingPathSwitch(Default=@n12, Deferred=@n12, Mobile=@n7) {desc=Migrated from FeatureLevelSwitch}
    B=ScalarParameter() {defaultValue=1, parameterName=Opacity}
   )
   B=@n9.A
  )
 ]
}
